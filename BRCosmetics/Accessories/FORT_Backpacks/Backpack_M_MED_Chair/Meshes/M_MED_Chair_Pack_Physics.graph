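# Physics asset for the M_MED_Chair_Pack backpack cosmetic, giving its dangling pillow secondary physics motion.
# A compound body on the pack bone, built from two capsules and a box, anchors a two-link chain of capsule bodies on the dyn_pillow bones.
# Tight swing and twist limits join the first pillow bone to the pack and the second to the first, keeping the sway small.
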PhysicsAsset /BRCosmetics/Accessories/FORT_Backpacks/Backpack_M_MED_Chair/Meshes/M_MED_Chair_Pack_Physics
bodies {
 dyn_pillow_01 [{shapeType=Capsule, length=3.1593, radius=3, center=<2.5 0.6502 0>, rotation=<rot 0 0 90>}]
 dyn_pillow_02 [{shapeType=Capsule, length=4.3454, radius=2.9334, center=<2.0859 0.0362 0>, rotation=<rot 0 -5.0001 89.9999>}]
 pack [
  {shapeType=Capsule, length=1, radius=1}
  {shapeType=Capsule, length=10.6669, radius=3.1162, center=<-10.7341 -11.9565 2.0967>, rotation=<rot 0 -5.0003 89.9996>}
  {shapeType=Box, extentX=20.6226, extentY=5.187, extentZ=18.6384, center=<-0.0002 -2.4896 5.2752>, rotation=<rot 0 0 10.0005>}
 ]
}
constraints [
 {bone1Name=dyn_pillow_01, bone2Name=pack, swing1LimitDegrees=20, swing2LimitDegrees=10, twistLimitDegrees=2}
 {bone1Name=dyn_pillow_02, bone2Name=dyn_pillow_01, swing1LimitDegrees=15, swing2LimitDegrees=10, twistLimitDegrees=2}
]
